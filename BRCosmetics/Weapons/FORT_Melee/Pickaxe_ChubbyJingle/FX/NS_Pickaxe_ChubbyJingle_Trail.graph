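# Niagara swing trail for the ChubbyJingle pickaxe cosmetic, pairing brown ribbon streaks with falling cookie chunks.
# Trail_Cam_Face and Trail_Side use AnimTrail modules to build short-lived ribbons that fade by emitter and particle age.
# CookieChunks spawns sprites at a weapon socket, colored by partition, inheriting velocity before gravity, drag and late shrinking take over.
# Time-of-day functions tint all emitters, while front-end modifiers adjust chunk size for menus versus gameplay.
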NiagaraSystem /BRCosmetics/Weapons/FORT_Melee/Pickaxe_ChubbyJingle/FX/NS_Pickaxe_ChubbyJingle_Trail
dependencies {
 usedDataInterfaces [
  &NiagaraDataInterfaceCurve
  &NiagaraDataInterfaceArrayColor
  &NiagaraDataInterfaceMeshRendererInfo
  &NiagaraDataInterfaceArrayInt32
  &NiagaraDataInterfaceSocketReader
 ]
 usedDynamicInputs []
 usedModules [
  &self:SystemSpawnScript.SystemScriptSource.SystemScriptGraph.NiagaraNodeAssignment_1.SetVariables_#1
  &SystemState
 ]
}
emitters {
 CookieChunks {
  emitterUpdateScript [
   &EmitterState()
   &SpawnRate(SpawnRate=15, Use Spawn Probability=false) {module=SpawnRate001}
  ]
  particleSpawnScript [
   &PartitionParticles(Number Of Partitions=3, Partition Mode:ENiagara_PartitionMode=::NewEnumerator5)
   &InitializeParticle(
    Color=&SelectLinearColorFromArray(
     Array Sampling Mode=::NewEnumerator1
     Direct Array Index=$Particles.PartitionParticles.Partition
     Color Selection Array=DataInterface() {propertyValues {ColorData [<rgba 0.3 0.0928 0.0172 1>, <rgba 0.27 0.0856 0.0183 1>, <rgba 0.29 0.0886 0.0151 1>, <rgba 0.9844 0.2931 0.041 1>]}}
     Random Seed=?
    )
    Lifetime Max=0.6
    Lifetime Min=0.3
    Lifetime Mode=::NewEnumerator1
    Mesh Scale Mode=::NewEnumerator3
    Mesh Uniform Scale=&Multiply_Float(
     A=0.4
     B=&FN_FrontEndFloatModifier(
      Gameplay=&RandomRangeFloat(Minimum=1, Maximum=2)
      Front End=&RandomRangeFloat(Minimum=0.5, Maximum=1)
     )
    )
    Sprite Rotation Mode=::NewEnumerator2
    Sprite Size Mode=::NewEnumerator3
    Sprite UV Mode=::NewEnumerator3
    Uniform Sprite Size=&Multiply_Float(
     A=0.6
     B=&FN_FrontEndFloatModifier(
      Gameplay=&UniformRangedFloat(Minimum=5, Maximum=10)
      Front End=&UniformRangedFloat(Minimum=3, Maximum=7)
     )
    )
    UsePositionOffset=false
    Write Material Random=true
   )
   /SetVariables(Particles.VisibilityTag=&SelectIntFromArray(Direct Array Index=$Particles.PartitionParticles.Partition, Int Selection Array=DataInterface() {propertyValues {IntData [0,0,0,1]}})) {isSetParameters=true}
   &SocketLocation(
    SamplePosOffset=true
    Sampled Position Offset=&RandomRangeVector(Minimum=<-5 -5 -5>, Maximum=<5 5 5>)
    Socket Index=2
    Socket Reader=$System.Socket Reader
    Socket Sampling Mode=::NewEnumerator3
    VelocitySpeedLimit=false
   )
   &InitialMeshOrientation(Use Rotation Vector=false)
   &ShapeLocation(MaskSpawnGroup=false, Override Seed=false, Sphere Radius=5, UseOwnerScale=true)
   &InheritVelocity(
    Inherited Velocity Amount Scale=&VectorFromFloat(Value=&RandomRangeFloat(Minimum=0.2, Maximum=1))
    Speed Limit=true
    UseSpeedThreshold=false
   ) {module=InheritVelocity001}
   &AddRotationalVelocity(Rotation Rate=&RandomVector(Vector Scale=&RandomRangeFloat(Minimum=1, Maximum=2)))
   &AddVelocity(
    Module.Use Falloff Distance=false
    Override Seed=false
    Velocity Mode=::NewEnumerator1
    Velocity Origin=$Particles.ShapeLocation.ShapePosition
    Velocity Speed=&RandomRangeFloat(Minimum=10, Maximum=20)
   )
   &ApplyInitialForces()
  ]
  particleUpdateScript [
   &ParticleState()
   &SpriteRotationRate(Rotation Rate=&RandomRangeFloat(Minimum=5, Maximum=25))
   &ScaleSpriteSize(Uniform Curve Sprite Scale=DataInterface() {propertyValues {Curve {keys [<at 0.6 1 RCIM_Cubic>, {arriveTangent=-6.308, interpMode=RCIM_Cubic, leaveTangent=-6.308, tangentMode=RCTM_User, time=1, value=0}]}}})
   &ScaleMeshSize(Scale Factor=&VectorFromFloat(Value=&FloatFromCurve(FloatCurve=DataInterface() {propertyValues {Curve {keys [<at 0.6 1 RCIM_Cubic>, {arriveTangent=-6.308, interpMode=RCIM_Cubic, leaveTangent=-6.308, tangentMode=RCTM_User, time=1, value=0}]}}})))
   &GravityForce(Gravity=&RandomRangeVector(Minimum=<0 0 -600>, Maximum=<0 0 -400>)) {module=GravityForce001}
   &Drag(Drag=1.5, Use Linear Drag=true, Use Rotational Drag=true)
   &SolveForcesAndVelocity(Clamp Velocity=false, Limit Acceleration=false)
   &ScaleColor(
    Scale RGB=&MakeVectorFromLinearColorRGB(Color=&FNTimeOfDayColor(
      Night Color=<rgba 0.7726 0.8287 1 1>
      Morning Color=<rgba 1 0.7838 0.6226 1>
      Day Color=<rgba 1 1 1 1>
      Evening Color=<rgba 1 0.691 0.4705 1>
     ))
    ScaleA=false
    ScaleRGB=true
   )
  ]
  rendererClasses [&NiagaraSpriteRendererProperties]
  renderers [{rendererClass=&NiagaraSpriteRendererProperties}]
 }
 Trail_Cam_Face {
  emitterUpdateScript [
   &EmitterState(Life Cycle Mode=::NewEnumerator1, Loop Behavior=::NewEnumerator1, Loop Duration=0.27, UseLoopDelay=false)
   &FN_SpawnAnimTrail_v2(AnimTrailWidthMode=::FromSecond)
  ]
  particleSpawnScript [
   /Game/Effects/Niagara/Modules/AnimTrail/FN_InitializeAnimTrail/FN_InitializeAnimTrailBoneIndecies(Second Bone or Socket Index=1)
   /Game/Effects/Niagara/Modules/AnimTrail/FN_InitializeAnimTrail/FN_InitializeAnimTrail_SocketReader/FN_InitializeAnimTrail_SocketReader(
    AnimTrailWidth=0.15
    Color:LinearColor=<rgba 0.55 0.1567 0.0487 0.5>
    Lifetime=0.22
    Socket Reader:NiagaraDataInterfaceSocketReader=$System.Socket Reader
   )
  ]
  particleUpdateScript [
   &FN_UpdateAnimTrail() {module=FN_UpdateAnimTrail001}
   &ScaleColor(
    Scale Alpha=&Multiply_Float(
     A=&FloatFromCurve(
      FloatCurve=DataInterface() {propertyValues {Curve {
         defaultValue=1
         keys [
          {arriveTangent=-1.5933, interpMode=RCIM_Cubic, leaveTangent=-1.5933, tangentMode=RCTM_User, time=0, value=1}
          {interpMode=RCIM_Cubic, tangentMode=RCTM_User, time=1, value=0}
         ]
        }}}
      Scale Curve=&FloatFromCurve(
       FloatCurve=DataInterface() {propertyValues {Curve {keys [
           {arriveTangent=22.4583, interpMode=RCIM_Cubic, leaveTangent=22.4583, tangentMode=RCTM_User, time=0, value=0}
           {arriveTangent=-0.0759, interpMode=RCIM_Cubic, leaveTangent=-0.0759, tangentMode=RCTM_User, time=0.1005, value=1}
           {arriveTangent=-0.3023, interpMode=RCIM_Cubic, leaveTangent=-0.3023, tangentMode=RCTM_User, time=0.15, value=1}
           {arriveTangent=0.0568, interpMode=RCIM_Cubic, leaveTangent=0.0568, tangentMode=RCTM_User, time=0.4, value=0}
          ]}}}
       CurveIndex=$Emitter.Age
      )
     )
     B=&FNTimeOfDayScalar(Night Scale=0.5, Morning Scale=0.6, Evening Scale=0.5)
    )
    Scale RGB=&MakeVectorFromLinearColorRGB(Color=&FNTimeOfDayColor(
      Night Color=<rgba 0.2059 0.3957 0.7188 1>
      Morning Color=<rgba 0.6 0.6262 0.8 1>
      Day Color=<rgba 1 1 1 1>
      Evening Color=<rgba 0.6 0.6 0.6 1>
     ))
    ScaleA=true
    ScaleRGB=true
   ) {module=ScaleColor002}
  ]
  rendererClasses [&NiagaraRibbonRendererProperties]
  renderers [{rendererClass=&NiagaraRibbonRendererProperties}]
 }
 Trail_Side {
  emitterSpawnScript [&FN_RibbonIDOffsetUtility()]
  emitterUpdateScript [
   &EmitterState()
   &FN_SpawnAnimTrail()
  ]
  particleSpawnScript [
   /Game/Effects/Niagara/Modules/AnimTrail/FN_InitializeAnimTrailBoneIndecies_v2/FN_InitializeAnimTrailBoneIndecies()
   &FN_InitializeAnimTrail(Color=<rgba 0.55 0.1567 0.0487 0.5>, Lifetime=0.22)
  ]
  particleUpdateScript [
   &FN_UpdateAnimTrail() {module=FN_UpdateAnimTrail001}
   &ScaleColor(
    Scale Alpha=&FloatFromCurve(
     FloatCurve=DataInterface() {propertyValues {Curve {keys [{interpMode=RCIM_Cubic, tangentMode=RCTM_User, time=0, value=1}, {interpMode=RCIM_Cubic, tangentMode=RCTM_User, time=0.7, value=0}]}}}
     Scale Curve=&FloatFromCurve(
      FloatCurve=DataInterface() {propertyValues {Curve {keys [
          {arriveTangent=22.4583, interpMode=RCIM_Cubic, leaveTangent=22.4583, tangentMode=RCTM_User, time=0, value=0}
          {arriveTangent=-0.0759, interpMode=RCIM_Cubic, leaveTangent=-0.0759, tangentMode=RCTM_User, time=0.1005, value=1}
          {arriveTangent=-0.3023, interpMode=RCIM_Cubic, leaveTangent=-0.3023, tangentMode=RCTM_User, time=0.15, value=1}
          {arriveTangent=0.0568, interpMode=RCIM_Cubic, leaveTangent=0.0568, tangentMode=RCTM_User, time=0.4, value=0}
         ]}}}
      CurveIndex=$Emitter.Age
      Scale Curve=&FN_FrontEndFloatModifier(Gameplay=&FNTimeOfDayScalar(Night Scale=1, Morning Scale=1, Evening Scale=1), Front End=1)
     )
    )
    Scale RGB=&MakeVectorFromLinearColorRGB(Color=&FNTimeOfDayColor(
      Night Color=<rgba 0.2059 0.3957 0.7188 1>
      Morning Color=<rgba 0.6 0.6262 0.8 1>
      Day Color=<rgba 1 1 1 1>
      Evening Color=<rgba 0.6 0.6 0.6 1>
     ))
    ScaleA=true
    ScaleRGB=true
   )
  ]
  rendererClasses [&NiagaraRibbonRendererProperties]
  renderers [{rendererClass=&NiagaraRibbonRendererProperties}]
 }
}
summary {
 systemName=NS_Pickaxe_ChubbyJingle_Trail
 emitters [
  {emitterName=Trail_Cam_Face, rendererClasses [&NiagaraRibbonRendererProperties]}
  {emitterName=Trail_Side, rendererClasses [&NiagaraRibbonRendererProperties]}
  {emitterName=CookieChunks, rendererClasses [&NiagaraSpriteRendererProperties]}
 ]
 userVariables [
  User.bIsFrontEnd {type=NiagaraBool}
  User.bIsFrontEndPreview {type=NiagaraBool}
  User.RibbonIDOffset {type=NiagaraInt32}
 ]
}
